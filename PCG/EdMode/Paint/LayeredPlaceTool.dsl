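; PCG Editor Mode paint tool graph that places meshes or actors from three painted, prioritized layers plus an exclusion layer.
; Each layer runs PlaceTool_BaseProcess with its own assets and transforms; the exclusion paint is then subtracted.
; A priority-sorted loop resolves layer overlaps, then L0_SpawnActors alone selects Static Mesh Spawner or Spawn Actor for all layers.
; Points tagged ShowPreviewMesh, brush stamps and painted bounds feed debug displays.
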(graph LayeredPlaceTool
  (params (L0_LayerPriority :type Int) (L0_UseMeshBounds :type Bool) (L0_StaticMeshes :type (Array (Struct /PCG/EdMode/Resources/PCGEdMode_WeightedStaticMeshReference.PCGEdMode_WeightedStaticMeshReference))) (L0_SpawnActors :type Bool) (L0_ActorClasses :type (Array (Struct /PCG/EdMode/Resources/PCGEdMode_WeightedClassReference.PCGEdMode_WeightedClassReference))) (L0_GlobalTransform :type (Struct /PCG/EdMode/Resources/PCGEdMode_GlobalTransformStruct.PCGEdMode_GlobalTransformStruct)) (L0_RandomTransform :type (Struct /PCG/EdMode/Resources/PCGEdMode_RandomTransformStruct.PCGEdMode_RandomTransformStruct)) (L1_LayerPriority :type Int) (L1_UseMeshBounds :type Bool) (L1_StaticMeshes :type (Array (Struct /PCG/EdMode/Resources/PCGEdMode_WeightedStaticMeshReference.PCGEdMode_WeightedStaticMeshReference))) (L1_SpawnActors :type Bool) (L1_ActorClasses :type (Array (Struct /PCG/EdMode/Resources/PCGEdMode_WeightedClassReference.PCGEdMode_WeightedClassReference))) (L1_GlobalTransform :type (Struct /PCG/EdMode/Resources/PCGEdMode_GlobalTransformStruct.PCGEdMode_GlobalTransformStruct)) (L1_RandomTransform :type (Struct /PCG/EdMode/Resources/PCGEdMode_RandomTransformStruct.PCGEdMode_RandomTransformStruct)) (L2_LayerPriority :type Int) (L2_UseMeshBounds :type Bool) (L2_StaticMeshes :type (Array (Struct /PCG/EdMode/Resources/PCGEdMode_WeightedStaticMeshReference.PCGEdMode_WeightedStaticMeshReference))) (L2_SpawnActors :type Bool) (L2_ActorClasses :type (Array (Struct /PCG/EdMode/Resources/PCGEdMode_WeightedClassReference.PCGEdMode_WeightedClassReference))) (L2_GlobalTransform :type (Struct /PCG/EdMode/Resources/PCGEdMode_GlobalTransformStruct.PCGEdMode_GlobalTransformStruct)) (L2_RandomTransform :type (Struct /PCG/EdMode/Resources/PCGEdMode_RandomTransformStruct.PCGEdMode_RandomTransformStruct)) (ShowBrushStamps :type Bool) (ShowPaintedDataBounds :type Bool))
  (bind Gather_0 (Gather))
  ;[0 0 0 0.5] Exclusion Layer
  (bind Difference_1 (Difference :bDiffMetadata false :DensityFunction Binary :Mode Discrete))
  (bind GetToolData_4 (Get%20Tool%20Data :DataInstance ExclusionLayer :ToolTag PaintTool))
  (bind FilterDataByType_3 (Filter%20Data%20By%20Type :TargetType (struct :Ids [(struct :Struct /Script/PCG.PCGDataTypeInfoPoint)])))
  ;[0 0 0 0.5] Difference by Priority and Spawning
  (bind SortTags_0 (Sort%20Data%20By%20Tag%20Value :Tag Priority))
  (bind Loop_0 (Loop :Graph /PCG/EdMode/Paint/Subgraphs/LayeredPlaceTool_DifferenceByPriority.LayeredPlaceTool_DifferenceByPriority))
  (bind FilterDataByTag_0 (Filter%20Data%20By%20Tag :SelectedTags "ShowPreviewMesh"))
  (bind Branch_1 (Branch :bOutputToB-param L0_SpawnActors))
  (bind StaticMeshSpawner_1 (Static%20Mesh%20Spawner :MeshSelectorParameters (class:/Script/PCG.PCGMeshSelectorByAttribute :AttributeName StaticMesh :TemplateDescriptor (struct :bUseDefaultCollision true)) :MeshSelectorType /Script/PCG.PCGMeshSelectorByAttribute))
  (bind SpawnActor_1 (Spawn%20Actor :bSpawnByAttribute true :PostSpawnFunctionNames [None] :SpawnAttribute Class))
  (bind Debug_2 (Debug))
  ;[1 0.1333 0] Debug Paint Data
  (bind Branch_3 (Branch :bOutputToB-param ShowBrushStamps))
  (bind AttributeMathsOp_5 (Attribute%20Maths%20Op :InputSource1 $Extents :Operation Divide :OutputTarget $Scale))
  (bind Branch_4 (Branch :bOutputToB-param ShowPaintedDataBounds))
  (bind MergePoints_0 (Merge%20Points :bMergeMetadata false))
  (bind CombinePoints_1 (Combine%20Points))
  (bind AttributeMathsOp_6 (Attribute%20Maths%20Op :InputSource1 $Extents :Operation Divide :OutputTarget $Scale))
  (bind Debug_0 (Debug))
  (bind Debug_1 (Debug))
  (bind Branch_0 (Branch :bOutputToB-param ShowBrushStamps))
  (bind AttributeMathsOp_0 (Attribute%20Maths%20Op :InputSource1 $Extents :Operation Divide :OutputTarget $Scale))
  (bind Debug_3 (Debug))
  ;[0 0 0 0.5] Layer 0
  (bind CopyAttributes_0 (Copy%20Attributes :bCopyAllAttributes true :InputSource @Last :Source L0_RandomTransform :Target L0_GlobalTransform))
  (bind GetToolData_1 (Get%20Tool%20Data :DataInstance Layer0 :ToolTag PaintTool))
  (bind FilterDataByType_1 (Filter%20Data%20By%20Type :TargetType (struct :Ids [(struct :Struct /Script/PCG.PCGDataTypeInfoPoint)])))
  (bind Subgraph_0 (Subgraph :Graph /PCG/EdMode/Paint/Subgraphs/PlaceTool_BaseProcess.PlaceTool_BaseProcess :ActorClasses L0_ActorClasses :LayerPriority L0_LayerPriority :SpawnActors L0_SpawnActors :StaticMeshes L0_StaticMeshes))
  ;[0 0 0 0.5] Layer 1
  (bind CopyAttributes_1 (Copy%20Attributes :bCopyAllAttributes true :InputSource @Last :Source L1_RandomTransform :Target L1_GlobalTransform))
  (bind GetToolData_0 (Get%20Tool%20Data :DataInstance Layer1 :ToolTag PaintTool))
  (bind FilterDataByType_0 (Filter%20Data%20By%20Type :TargetType (struct :Ids [(struct :Struct /Script/PCG.PCGDataTypeInfoPoint)])))
  (bind Subgraph_1 (Subgraph :Graph /PCG/EdMode/Paint/Subgraphs/PlaceTool_BaseProcess.PlaceTool_BaseProcess :ActorClasses L1_ActorClasses :LayerPriority L1_LayerPriority :SpawnActors L1_SpawnActors :StaticMeshes L1_StaticMeshes))
  ;[0 0 0 0.5] Layer 2
  (bind GetToolData_2 (Get%20Tool%20Data :DataInstance Layer2 :ToolTag PaintTool))
  (bind FilterDataByType_2 (Filter%20Data%20By%20Type :TargetType (struct :Ids [(struct :Struct /Script/PCG.PCGDataTypeInfoPoint)])))
  (bind CopyAttributes_2 (Copy%20Attributes :bCopyAllAttributes true :InputSource @Last :Source L2_RandomTransform :Target L2_GlobalTransform))
  (bind Subgraph_2 (Subgraph :Graph /PCG/EdMode/Paint/Subgraphs/PlaceTool_BaseProcess.PlaceTool_BaseProcess :ActorClasses L2_ActorClasses :LayerPriority L2_LayerPriority :SpawnActors L2_SpawnActors :StaticMeshes L2_StaticMeshes))
  (reroute ExclusionLayerPaintData FilterDataByType_3.InsideFilter)
  (reroute Layer%200%20Data Subgraph_0)
  (reroute Layer%200%20Paint%20Data FilterDataByType_1.InsideFilter)
  (reroute Layer%201%20Data Subgraph_1)
  (reroute Layer%201%20Paint%20Data FilterDataByType_0.InsideFilter)
  (reroute Layer%202%20Data Subgraph_2)
  (reroute Layer%202%20Paint%20Data FilterDataByType_2.InsideFilter)
  (connect Branch_0 Output%20B AttributeMathsOp_0 InA)
  (connect Branch_3 Output%20B AttributeMathsOp_5 InA)
  (connect CombinePoints_1 Out AttributeMathsOp_6 InA)
  (connect ExclusionLayerPaintData Out Branch_0 In)
  (connect FilterDataByTag_0 OutsideFilter Branch_1 In)
  (connect Layer%200%20Paint%20Data Out Branch_3 In)
  (connect Layer%201%20Paint%20Data Out Branch_3 In)
  (connect Layer%202%20Paint%20Data Out Branch_3 In)
  (connect Layer%200%20Paint%20Data Out Branch_4 In)
  (connect Layer%201%20Paint%20Data Out Branch_4 In)
  (connect Layer%202%20Paint%20Data Out Branch_4 In)
  (connect ExclusionLayerPaintData Out Branch_4 In)
  (connect MergePoints_0 Out CombinePoints_1 In)
  (connect AttributeMathsOp_5 Out Debug_0 In)
  (connect AttributeMathsOp_6 Out Debug_1 In)
  (connect FilterDataByTag_0 InsideFilter Debug_2 In)
  (connect AttributeMathsOp_0 Out Debug_3 In)
  (connect ExclusionLayerPaintData Out Difference_1 Differences)
  (connect Gather_0 Out Difference_1 Source)
  (connect Loop_0 RemainingPoints FilterDataByTag_0 In)
  (connect GetToolData_0 Out FilterDataByType_0 In)
  (connect GetToolData_1 Out FilterDataByType_1 In)
  (connect GetToolData_2 Out FilterDataByType_2 In)
  (connect GetToolData_4 Out FilterDataByType_3 In)
  (connect Layer%200%20Data Out Gather_0 In)
  (connect Layer%201%20Data Out Gather_0 In2)
  (connect Layer%202%20Data Out Gather_0 In3)
  (connect SortTags_0 Out Loop_0 SourcePoints)
  (connect Branch_4 Output%20B MergePoints_0 In)
  (connect Difference_1 Out SortTags_0 In)
  (connect Branch_1 Output%20B SpawnActor_1 In)
  (connect Branch_1 Output%20A StaticMeshSpawner_1 In)
  (connect CopyAttributes_0 Out Subgraph_0 Overrides)
  (connect FilterDataByType_1 InsideFilter Subgraph_0 ToolData)
  (connect CopyAttributes_1 Out Subgraph_1 Overrides)
  (connect FilterDataByType_0 InsideFilter Subgraph_1 ToolData)
  (connect CopyAttributes_2 Out Subgraph_2 Overrides)
  (connect FilterDataByType_2 InsideFilter Subgraph_2 ToolData))
