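; PCG editor-mode paint tool graph that places weighted static meshes or actors on points painted on its paint layer.
; A base-process subgraph receives the points plus merged global and random transform settings.
; Results tagged ShowPreviewMesh go to a debug node; the rest branch on SpawnActors into attribute-driven mesh or actor spawners.
; Optional debug branches display brush stamps and combined bounds of the painted data.
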(graph PlaceTool
  (params (StaticMeshes :type (Array (Struct /PCG/EdMode/Resources/PCGEdMode_WeightedStaticMeshReference.PCGEdMode_WeightedStaticMeshReference))) (SpawnActors :type Bool) (ActorClasses :type (Array (Struct /PCG/EdMode/Resources/PCGEdMode_WeightedClassReference.PCGEdMode_WeightedClassReference))) (GlobalTransform :type (Struct /PCG/EdMode/Resources/PCGEdMode_GlobalTransformStruct.PCGEdMode_GlobalTransformStruct)) (RandomTransform :type (Struct /PCG/EdMode/Resources/PCGEdMode_RandomTransformStruct.PCGEdMode_RandomTransformStruct)) (ShowBrushStamps :type Bool) (ShowPaintedDataBounds :type Bool))
  (bind CopyAttributes_0 (Copy%20Attributes :bCopyAllAttributes true :InputSource @Last :Source RandomTransform :Target GlobalTransform))
  (bind GetToolData_1 (Get%20Tool%20Data :DataInstance PaintLayer :ToolTag PaintTool))
  (bind FilterDataByType_1 (Filter%20Data%20By%20Type :TargetType (struct :Ids [(struct :Struct /Script/PCG.PCGDataTypeInfoPoint)])))
  (bind Subgraph_0 (Subgraph :Graph /PCG/EdMode/Paint/Subgraphs/PlaceTool_BaseProcess.PlaceTool_BaseProcess :ActorClasses ActorClasses :SpawnActors SpawnActors :StaticMeshes StaticMeshes))
  (bind FilterDataByTag_0 (Filter%20Data%20By%20Tag :SelectedTags "ShowPreviewMesh"))
  (bind Branch_1 (Branch :bOutputToB-param SpawnActors))
  (bind StaticMeshSpawner_1 (Static%20Mesh%20Spawner :MeshSelectorParameters (class:/Script/PCG.PCGMeshSelectorByAttribute :AttributeName StaticMesh :TemplateDescriptor (struct :bUseDefaultCollision true)) :MeshSelectorType /Script/PCG.PCGMeshSelectorByAttribute))
  (bind SpawnActor_1 (Spawn%20Actor :bSpawnByAttribute true :PostSpawnFunctionNames [None] :SpawnAttribute Class))
  (bind Debug_3 (Debug))
  ;[1 0.1333 0] Debug Paint Data
  (bind Branch_3 (Branch :bOutputToB-param ShowBrushStamps))
  (bind AttributeMathsOp_5 (Attribute%20Maths%20Op :InputSource1 $Extents :Operation Divide :OutputTarget $Scale))
  (bind Branch_4 (Branch :bOutputToB-param ShowPaintedDataBounds))
  (bind CombinePoints_1 (Combine%20Points))
  (bind AttributeMathsOp_6 (Attribute%20Maths%20Op :InputSource1 $Extents :Operation Divide :OutputTarget $Scale))
  (bind Debug_0 (Debug))
  (bind Debug_1 (Debug))
  (reroute Paint%20Layer%20Data FilterDataByType_1.InsideFilter)
  (connect Branch_3 Output%20B AttributeMathsOp_5 InA)
  (connect CombinePoints_1 Out AttributeMathsOp_6 InA)
  (connect FilterDataByTag_0 OutsideFilter Branch_1 In)
  (connect Paint%20Layer%20Data Out Branch_3 In)
  (connect Paint%20Layer%20Data Out Branch_4 In)
  (connect Branch_4 Output%20B CombinePoints_1 In)
  (connect AttributeMathsOp_5 Out Debug_0 In)
  (connect AttributeMathsOp_6 Out Debug_1 In)
  (connect FilterDataByTag_0 InsideFilter Debug_3 In)
  (connect Subgraph_0 Out FilterDataByTag_0 In)
  (connect GetToolData_1 Out FilterDataByType_1 In)
  (connect Branch_1 Output%20B SpawnActor_1 In)
  (connect Branch_1 Output%20A StaticMeshSpawner_1 In)
  (connect CopyAttributes_0 Out Subgraph_0 Overrides)
  (connect FilterDataByType_1 InsideFilter Subgraph_0 ToolData))
